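# Material for the Degree Inverted vehicle wheel, adding an animated emissive glow to the simple wheel parent function.
# Two panning samples of a geometric texture combine, one floored, into a clamped pulse that multiplies the tire and tread curvature masks to drive the emissive color.
# Vertex color red blends between tire and tread glow.
# The result passes through the shared vehicle damage and de-rez transition functions.
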Material /VehicleCosmetics/Wheels/Degree_Inverted/Materials/M_Wheel_DegreeInverted
graph {
 properties {
  bAutomaticallySetUsageInEditor=false
  bUseMaterialAttributes=true
  bUsedWithMeshDeformer=true
  bUsedWithSkeletalMesh=true
  pixelDepthOffsetMode=PDOM_Legacy
  usedShadingModels=Default Lit
 }
 nodes [
  n1 = MaterialFunctionCall() {materialFunction=&MF_WheelParent_Simple}
  n2 = VectorParameter() {group=5_Special Effects, parameterName=Emissive Color, sortPriority=0, defaultValue=<rgba 0.1296 0.2054 1 1>}
  n3 = TextureSample(UVs=Panner(Coordinate=TextureCoordinate() {uTiling=0.33}) {speedX=0.5}) {texture=/VehicleCosmetics/Wheels/_SharedTextures/FX/Skin_GeometricBM_RGB}
  n4 = Panner() {speedX=0.4}
  n5 = Clamp(None=Add(A=Floor(None=@n3.R), B=TextureSample(UVs=@n4) {texture=&Skin_GeometricBM_RGB}.R)) {maxDefault=3}
  n6 = LinearInterpolate(A=@n5, B=@n5, Alpha=NamedRerouteUsage())
  n7 = TextureSampleParameter2D() {
   group=5_Special Effects
   parameterName=Tire Curvature Mask
   sortPriority=-2
   texture=/VehicleCosmetics/Wheels/Degree_Inverted/Textures/Wheel_degree_SE_Curv
  }
  n8 = TextureSampleParameter2D() {
   group=5_Special Effects
   parameterName=Tread Curvature Mask
   sortPriority=-1
   texture=/VehicleCosmetics/Wheels/Degree_Inverted/Textures/Degree_SE_Fan_Curv
  }
  n9 = LinearInterpolate(
   A=LinearInterpolate(B=@n2.RGB, Alpha=Multiply(A=@n6, B=@n7.R))
   B=LinearInterpolate(B=@n2.RGB, Alpha=Multiply(A=@n6, B=@n8.R))
   Alpha=NamedRerouteUsage()
  )
  MaterialFunctionCall(In=MaterialFunctionCall(Material=SetMaterialAttributes(MaterialAttributes=@n1.MaterialAttributes, Emissive Color=@n9) {attributeSetTypes [#1]}) {materialFunction=&MF_BaseVehicle_Damage}.Result) {materialFunction=&MF_VehicleDeRez_Transition}
  NamedRerouteDeclaration(None=ComponentMask(None=@n1.VertexColorR) {r=true}) {name=VC_R, nodeColor=<rgba 0.875 1 0 1>}
 ]
}
